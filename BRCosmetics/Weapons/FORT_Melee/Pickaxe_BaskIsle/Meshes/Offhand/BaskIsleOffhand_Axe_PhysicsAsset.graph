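# Physics asset for the offhand axe of the BaskIsle pickaxe cosmetic, a melee weapon.
# It holds a single body on the Melee_TwoHanded_Handle bone, shaped as one capsule offset slightly from the bone origin.
# It has no constraints, so the offhand axe is represented as one collision piece.
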PhysicsAsset /BRCosmetics/Weapons/FORT_Melee/Pickaxe_BaskIsle/Meshes/Offhand/BaskIsleOffhand_Axe_PhysicsAsset
bodies {Melee_TwoHanded_Handle [{shapeType=Capsule, length=15.1996, radius=6.6459, center=<0.0319 -0.1413 1.9172>, rotation=<rot -0.0001 -90 -0.2951>}]}
constraints []
